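# Transition material for the Fire Petal shrine's clay-tile boon obelisk, adding animated glow effects to a four-layer Nanite master material.
# Four SFX-maskable layers merge through MF_Layer_4to1, then optional SSS edge depth, detail, sticker, snow and leak passes.
# Custom primitive TimeInput and Duration derive animation progress, alongside noise-distorted rising rings, gradual fill and sphere-masked sprite/logo icon emissives.
# A rune mask darkens base color; Activated blends in a sine wobble offset.
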Material /FirePetalShrineGameplay/Assets/Materials/M_ClayTile_BoonObelisk_A_Transition
graph {
 properties {
  bAutomaticallySetUsageInEditor=false
  bUseMaterialAttributes=true
  bUsedWithNanite=true
  shadingModel=MSM_FromMaterialExpression
  usedShadingModels=Default Lit | Subsurface | Clear Coat | Two Sided Foliage
 }
 nodes [
  n1 = MaterialFunctionCall!(SFX V3=NamedRerouteUsage(), 1=NamedRerouteUsage(), 2=NamedRerouteUsage(), 3=NamedRerouteUsage(), 4=NamedRerouteUsage()) {materialFunction=/Game/Global/Base4Layers_2022/MaterialFunctions/MF_Layer_4to1}
  Composite!(None=@n1.Result)
  Composite!() {subgraphName=L2 Blending Logic help}
  Composite() {subgraphName=L3 Blending Logic help}
  Composite() {subgraphName=L4 Blending Logic help}
  Constant3Vector() {constant=<rgba 5.7218 6 3.1875 1>}
  n2 = Time()
  n3 = ScalarParameter() {defaultValue=15, parameterName=Duration}
  ConstantBiasScale(None=Frac(None=Divide(A=@n2, B=@n3))) {bias=-0.25, scale=1.5}
  Constant() {r=1}
  Frac(None=Multiply(A=Time()) {constB=0.5})
  n4 = NamedRerouteUsage()
  n5 = SmoothStep(Value=SphereMask(A=NamedRerouteUsage(), B=Constant3Vector() {constant=<rgba 3 60 255 0>}) {attenuationRadius=100, hardnessPercent=10}) {constMax=0.45}
  LinearInterpolate(A=@n4, Alpha=@n5) {constB=0}
  MaterialFunctionCall() {materialFunction=&DebugTimeSine}
  n6 = Divide(A=ComponentMask(None=NamedRerouteUsage()) {b=true}) {constB=300}
  n7 = VectorParameter() {parameterName=RingEmissive, defaultValue=<rgba 0 0 0 1>}
  MaterialFunctionCall(
   Hue Shift Percentage=LinearInterpolate(A=ScalarParameter() {defaultValue=0.2, parameterName=HueshiftAmount}, Alpha=Saturate(None=@n6)) {constB=0}
   Texture=@n7.RGB
  ) {materialFunction=&HueShift}
  n8 = TextureObjectParameter() {
   desc="SFX TextureGrayscale You can reuse that texture across layers and functions. "
   group=0_SFX Global
   parameterName=SFX_Grayscale
   samplerType=SAMPLERTYPE_LinearGrayscale
   sortPriority=0
   texture=/Game/Global/Base4Layers_2022/Textures/T_SFX_Grayscale
  }
  n9 = TextureObjectParameter() {
   desc="SFX UniversalTextureMask Uses Dot to select channel used for various parts of material.You can reuse that texture across layers and functions. "
   group=0_SFX Global
   parameterName=SFX_Mask
   samplerType=SAMPLERTYPE_Masks
   sortPriority=0
   texture=/Game/Global/Base4Layers_2022/Textures/T_SFX_M
  }
  n10 = StaticSwitchParameter(True=@n8, False=@n9) {desc=Make sure size is less than 512 and lower, group=0_SFX Global, parameterName=UseSFXGrayscale}
  NamedRerouteDeclaration(None=MaterialFunctionCall!(In=@n10) {materialFunction=/Game/Global/Base4Layers_2022/MaterialFunctions/MF_SFXTexture}.SFX) {name=SFX V3, nodeColor=<rgba 0.6406 0 1 1>}
  n11 = NamedRerouteUsage()
  n12 = MaterialFunctionCall!(
   ColorTex=TextureObject() {texture=/Game/Tools/WaterStreamGenerator/Tmp/T_CausticsNoise}
   NormalTex
   Pos
   Nrm
   ColorUVScale=ScalarParameter() {defaultValue=100, parameterName=NoiseScale}
   ColorBlendFactor
   NormalStrength
   NormalUVScale
   NormalBlendFactor
   Local Space Inputs
  ) {materialFunction=/Game/Global/MaterialFunctions/MF_BiPlanar}
  n13 = Subtract(
   A=Divide(A=LocalPosition().Z, B=ScalarParameter() {defaultValue=700, parameterName=HeightDivision})
   B=Multiply(
    A=NamedRerouteDeclaration(None=@n12.Color) {name=Noise, nodeColor=<rgba 0 1 0.25 1>}
    B=ScalarParameter() {defaultValue=0.1, parameterName=NoiseStrength}
   )
  )
  n14 = Multiply(A=Time(), B=ScalarParameter() {defaultValue=0.68, parameterName=RingSpeed})
  n15 = MaterialFunctionCall(Float3=NamedRerouteUsage()) {materialFunction=&BreakOutFloat3Components}
  n16 = MaterialFunctionCall!(1=@n15.G, 2=@n15.G, 3, 4) {materialFunction=/Game/Global/Base4Layers_2022/MaterialFunctions/MF_UV2LerpScalar_Layers_New}
  n17 = Multiply(
   A=SphereMask(A=Saturate(None=@n13), B=Frac(None=@n14), Radius=ScalarParameter() {defaultValue=0.17, parameterName=RingRadius}) {hardnessPercent=0}
   B=@n16.Result
  )
  NamedRerouteDeclaration(None=Multiply(A=ScalarParameter() {defaultValue=0.02, parameterName=EmissiveMult}, B=Multiply(A=@n11, B=@n17))) {name=Ring_Emissive, nodeColor=<rgba 1 0.5156 0 1>}
  n18 = LocalPosition() {includedOffsets=ExcludeOffsets}
  n19 = MaterialFunctionCall(Float3=@n18.XYZ) {materialFunction=&BreakOutFloat3Components}
  NamedRerouteDeclaration(None=@n19.B) {name=Top_Height, nodeColor=<rgba 0.2656 1 0 1>}
  n20 = ScalarParameter() {defaultValue=100, parameterName=Param_2}
  n21 = AppendVector(
   A=SphereMask(A=@n19.R, B=ScalarParameter() {defaultValue=-50}, Radius=@n20) {attenuationRadius=180, hardnessPercent=0}
   B=SphereMask(A=@n19.B, B=ScalarParameter() {defaultValue=200, parameterName=Param_1}, Radius=@n20) {attenuationRadius=180, hardnessPercent=0}
  )
  n22 = TextureSample(UVs=@n21) {samplerSource=SSM_Wrap_WorldGroupSettings, samplerType=SAMPLERTYPE_Masks, texture=/FirePetalShrineGameplay/Assets/Textures/T_Sprite_Water}
  n23 = NamedRerouteUsage()
  n24 = Multiply(A=@n23) {constB=0.05}
  n25 = ScalarParameter() {defaultValue=0.5, parameterName=IconThreshHold}
  n26 = MaterialFunctionCall(Float2=@n21) {materialFunction=&BreakOutFloat2Components}
  n27 = ComponentMask(None=@n18.XYZ) {g=true}
  n28 = LinearInterpolate(
   A=@n21
   B=MaterialFunctionCall(X=Multiply(A=@n26.R) {constB=-1}, Y=@n26.G) {materialFunction=&MakeFloat2}.Result
   Alpha=Saturate(None=Sign(None=@n27))
  )
  n29 = LinearInterpolate(
   A=LinearInterpolate(
    A=TextureSample(UVs=@n21) {samplerSource=SSM_Wrap_WorldGroupSettings, texture=/FirePetalShrineGameplay/Assets/Textures/T_Sprite_Wind_Icon}.R
    B=@n22.R
    Alpha=Step(Y=@n24, X=@n25) {constX=0.3}
   )
   B=TextureSample(UVs=@n28) {samplerType=SAMPLERTYPE_Masks, texture=/FirePetalShrineGameplay/Assets/Textures/T_Sprite_Hero_Shrine}.R
   Alpha=Step(X=ComponentMask(None=@n23) {r=true}) {constX=25, constY=25}
  )
  n30 = TextureSample(UVs=@n28) {texture=/FirePetalShrineGameplay/Assets/Textures/T_Supermanlogo_Alpha}
  n31 = Step(X=ComponentMask(None=@n23) {b=true}) {constY=25}
  n32 = NamedRerouteUsage()
  n33 = MaterialFunctionCall(X=@n19.R, Y=Abs(None=@n19.G), Z=@n19.R) {materialFunction=&MakeFloat3}
  n34 = Multiply(
   A=LinearInterpolate(B=LinearInterpolate(A=@n29, B=@n30.RGB, Alpha=@n31), Alpha=Sign(None=@n32))
   B=SphereMask(A=@n33.Result, B=Constant3Vector() {constant=<rgba 3 50 255 0>}) {attenuationRadius=77}
  )
  n35 = ScalarParameter() {defaultValue=0.5, parameterName=IconEmissiveMult}
  n36 = NamedRerouteUsage()
  n37 = TextureSample(UVs=@n28) {texture=/FirePetalShrineGameplay/Assets/Textures/T_Supermanlogo}
  n38 = ScalarParameter() {defaultValue=10, parameterName=StandardSupermanIntensity}
  n39 = ComponentMask(None=@n36) {b=true}
  n40 = NamedRerouteUsage()
  n41 = Add(A=Constant() {r=0.4}, B=Multiply(A=NamedRerouteUsage()) {constA=1, constB=0.21})
  n42 = Multiply(A=Saturate(None=NamedRerouteUsage()), B=ScalarParameter() {defaultValue=0.02, parameterName=FillingNoise})
  n43 = Multiply(
   A=LinearInterpolate(A=@n36, B=Multiply(A=@n37.RGB, B=@n38), Alpha=Step(X=@n39) {constY=25})
   B=SphereMask(A=Divide(A=@n40) {constB=500}, B=Constant(), Radius=Add(A=@n41, B=@n42) {constB=0}) {attenuationRadius=0.63, hardnessPercent=30}
  )
  NamedRerouteDeclaration(None=Multiply(A=NamedRerouteDeclaration(None=@n34) {name=WaterIcon, nodeColor=<rgba 0.0547 0 1 1>}, B=Multiply(A=@n35, B=@n43))) {name=Icon_Emissive, nodeColor=<rgba 1 0 0.5859 1>}
  NamedRerouteDeclaration(None=@n7.RGB) {name=EmissiveInput, nodeColor=<rgba 0.4531 1 0 1>}
  NamedRerouteDeclaration(None=@n33.Result) {name=LocalPosition, nodeColor=<rgba 1 0.7031 0 1>}
  n44 = Transform(None=NamedRerouteUsage())
  n45 = StaticBool() {value=true}
  n46 = ScalarParameter() {defaultValue=0.1, parameterName=FillingGradualEmissiveMult}
  n47 = Multiply(A=NamedRerouteUsage(), B=ScalarParameter() {defaultValue=80, parameterName=FillingGradualNoiseAmount})
  n48 = NamedRerouteUsage()
  n49 = Multiply(A=Multiply(A=Power(Base=@n48) {constExponent=1.3}) {constB=1000}, B=@n16.Result)
  n50 = Multiply(
   A=NamedRerouteUsage()
   B=SphereMask(A=NamedRerouteUsage(), B=Constant3Vector() {constant=<rgba 3 50 50 0>}, Radius=Add(A=@n47, B=@n49)) {attenuationRadius=300, hardnessPercent=40}
  )
  NamedRerouteDeclaration(None=Multiply(
    A=MaterialFunctionCall(Normal Vector=@n44, Invert Fresnel=@n45, Use Cheap contrast=@n45) {materialFunction=&Fresnel_Function}.Result
    B=Multiply(A=@n46, B=@n50)
   )) {name=FillingGradual, nodeColor=<rgba 1 0.5625 0 1>}
  n51 = NamedRerouteUsage()
  n52 = NamedRerouteUsage()
  MaterialExpressionReroute_3 = Reroute(None=@n1.Result)
  n53 = GetMaterialAttributes(None=@MaterialExpressionReroute_3.Input) {attributeGetTypes [#1, #2]}
  n54 = StaticSwitchParameter(
   True=StaticSwitchParameter(True=TextureCoordinate() {coordinateIndex=2}, False=TextureCoordinate() {coordinateIndex=1}) {group=B_SSS_EdgeDepth, parameterName=SSS Depth UV2, sortPriority=-30}
   False=TextureCoordinate()
  ) {desc="Default is UV0 ", group=B_SSS_EdgeDepth, parameterName=SSS Depth UV1, sortPriority=-30}
  n55 = AppendVector(
   A=ScalarParameter() {defaultValue=1, group=B_SSS_EdgeDepth, parameterName=SSSDepth X, sortPriority=5}
   B=ScalarParameter() {defaultValue=1, group=B_SSS_EdgeDepth, parameterName=SSSDepth Y, sortPriority=5}
  )
  n56 = Add(
   A=ComponentMask(None=MaterialFunctionCall!(UVs=@n54, Tiling=@n55, Randomize Rotation) {materialFunction=/Engine/Functions/Engine_MaterialFunctions01/Cubemaps/InteriorCubemap}.UVW) {r=true}
   B=Constant() {r=0.98}
  )
  n57 = ScalarParameter() {defaultValue=3, group=B_SSS_EdgeDepth, parameterName=SSS_Depth Edge Blur, sortPriority=3}
  n58 = LinearInterpolate(A=ScalarParameter() {group=B_SSS_EdgeDepth, parameterName=SSS Depth Min, sortPriority=4}, Alpha=Saturate(None=Divide(A=@n56, B=@n57)))
  n59 = ScalarParameter() {group=B_SSS_EdgeDepth, parameterName=SSS Fresnel Min, sortPriority=1}
  n60 = Saturate(None=Fresnel(
    ExponentIn=ScalarParameter() {defaultValue=3, group=B_SSS_EdgeDepth, parameterName=SSS Fresnel Exponent, sortPriority=2}
    Normal=Transform(None=@n53.Normal)
   ) {baseReflectFraction=0, exponent=3})
  n61 = Saturate(None=Multiply(A=Multiply(A=@n53.Subsurface Color, B=@n58), B=LinearInterpolate(B=@n59, Alpha=@n60) {constA=1}))
  n62 = StaticSwitchParameter(
   True=QualitySwitch(
    Default=@MaterialExpressionReroute_3.Input
    Epic=SetMaterialAttributes!(MaterialAttributes=@n53.MaterialAttributes, Subsurface Color=@n61) {attributeSetTypes [#1]}
   )
   False=@MaterialExpressionReroute_3.Input
  ) {group=B_SSS_EdgeDepth, parameterName=SSS_EdgeDepth}
  MaterialExpressionReroute_9 = Reroute(None=@n62)
  n63 = NamedRerouteUsage()
  n64 = MaterialFunctionCall!(
   SFX V3=@n51
   Mat In=MaterialFunctionCall!(
    Material In=MaterialFunctionCall!(SFX=@n52, Material In=@MaterialExpressionReroute_9.Output, SFX Textures RGB=@n63) {materialFunction=/Game/Global/Base4Layers_2022/MaterialFunctions/MF_EmissiveControls_2}.Result
    SFX Texture=NamedRerouteUsage()
    SFX RGB=NamedRerouteUsage()
   ) {materialFunction=/Game/Global/Base4Layers_2022/MaterialFunctions/MF_StickersPrint}.Result
  ) {materialFunction=/Game/Global/Base4Layers_2022/MaterialFunctions/MF_DetailsWS}
  n65 = MaterialFunctionCall!(
   AO=@n64.AO Mask
   In=MaterialFunctionCall!(Mat In=@n64.Result) {materialFunction=/Game/Global/Base4Layers_2022/MaterialFunctions/MF_SnowWS}.Result
  ) {materialFunction=/Game/Global/Base4Layers_2022/MaterialFunctions/MF_Leaks}
  n66 = NamedRerouteUsage()
  n67 = MaterialFunctionCall!(
   Layer Debug
   Debug Layer Color
   Mat=MaterialFunctionCall!(
    In=MaterialFunctionCall!(In=@n65.Result, SFXMask=@n66) {materialFunction=/Game/Global/Base4Layers_2022/MaterialFunctions/MF_4Layer_ThinFilm}.Result
    SFXMask=@n66
   ) {materialFunction=/Game/Global/Base4Layers_2022/MaterialFunctions/MF_DiffuseHueShift}.Result
  ) {materialFunction=/Game/Global/Base4Layers_2022/MaterialFunctions/MF_DebugViews}
  n68 = GetMaterialAttributes(None=@n67.POI Color) {attributeGetTypes [#3, #2]}
  NamedRerouteDeclaration(None=@n68.Normal) {name=Base_Normal, nodeColor=<rgba 1 0 0.1875 1>}
  n69 = Constant()
  n70 = Multiply(
   A=MaterialFunctionCall(Speed=ScalarParameter() {defaultValue=0.3, parameterName=WobbleSpeed}) {materialFunction=&DebugTimeSine}.Result
   B=ScalarParameter() {defaultValue=15, parameterName=WobbleAmount}
  )
  NamedRerouteDeclaration(None=LinearInterpolate(
    B=MaterialFunctionCall(X=@n69, Y=@n69, Z=@n70) {materialFunction=&MakeFloat3}.Result
    Alpha=ScalarParameter() {parameterName=Activated}
   ) {constB=0}) {name=WPO, nodeColor=<rgba 0 0.6484 1 1>}
  n71 = StaticBoolParameter() {group=4_layer UV, parameterName=Use UV2 for Layer 4}
  n72 = MaterialFunctionCall!(L3 UV2, L2 UV2, L4 UV2=@n71) {materialFunction=/Game/Global/Base4Layers_2022/MaterialFunctions/MF_UVsSet_New}
  n73 = StaticBoolParameter() {desc=Enable to use World Space Mapping for layer. Only Z projection available in Layer 4, group=4_layer UV, parameterName=Layer 4 WS Mapping}
  n74 = StaticBoolParameter() {
   desc="World Space Z only, you need appropriate mask to make it visible. Use Z or map your layers accordingly "
   group=3_layer UV
   parameterName=Layer 3 WSM_Z
  }
  n75 = StaticBool() {value=true}
  n76 = StaticBool()
  n77 = StaticBoolParameter() {group=4_Layer_Blending, parameterName=Blend Overlay Layer 4}
  n78 = StaticSwitch(
   True=StaticBoolParameter() {desc=Use SFX to mask needed area, group=4_Layer_Blending, parameterName=L4 Simple Swatch, sortPriority=-15}
   False=@n76
   Value=@n77
  )
  MaterialExpressionReroute_16 = Reroute(None=@n78)
  n79 = StaticBoolParameter() {desc="World Space XYZ ( Expensive)", group=3_layer UV, parameterName=Layer 3 WSM_XYZ}
  n80 = StaticBool()
  n81 = TextureObjectParameter() {
   group=4_Layer_Nanite
   parameterName="Normals  Nanite 4"
   samplerType=SAMPLERTYPE_Normal
   sortPriority=0
   texture=/Engine/EngineMaterials/FlatNormal
  }
  n82 = TextureObjectParameter() {
   group=4_Layer_Nanite
   parameterName="SpecularMasks  Nanite 4"
   samplerType=SAMPLERTYPE_Masks
   sortPriority=0
   texture=&T_Fortnite_Default_S
  }
  n83 = TextureObjectParameter() {group=4_Layer_Nanite, parameterName=Diffuse Nanite 4, sortPriority=0, texture=/Game/Global/Textures/Default/Blanks/T_MidgreysRGB}
  n84 = StaticBoolParameter() {desc="Enables SFX for Opacity Mask, mast have Masked blending set in MI", group=4_Layer SFX, parameterName=SFX OpacityMask 4, sortPriority=1}
  n85 = StaticSwitch(True=StaticBoolParameter() {group=4_Layer_Blending, parameterName="L4  Normals Only", sortPriority=-15}, False=@n76, Value=@n77)
  MaterialExpressionReroute_17 = Reroute(None=@n85)
  n86 = TextureObjectParameter() {group=4_Layer, parameterName=Diffuse_Texture_4, sortPriority=0, texture=&T_MidgreysRGB}
  n87 = TextureObjectParameter() {
   desc="R = Specular\r\nG = Metal\r\nB = Roughness"
   group=4_Layer
   parameterName=SpecularMasks_4
   samplerType=SAMPLERTYPE_Masks
   sortPriority=0
   texture=&T_Fortnite_Default_S
  }
  n88 = TextureObjectParameter() {group=4_Layer, parameterName=Normals_Texture_4, samplerType=SAMPLERTYPE_Normal, sortPriority=0, texture=&FlatNormal}
  n89 = VectorParameter() {
   group=4_Layer SFX
   parameterName="SFX_SSS Tint(Expand)_4"
   primitiveDataIndex=2
   sortPriority=0
   channelNames {a=SSS Depth 0-1, b="", g="", r=""}
  }
  n90 = VectorParameter() {
   desc="Use R and G channels of this input to controll ClearCoat amount and Rounghness. It uses SFX texture to mask area affected. "
   group=4_Layer SFX
   parameterName="SFX_ClearCoat(Expand)_4"
   sortPriority=-3
   channelNames {a=Not Used, b=Not Used, g=Clearcoat Roughness, r=Clearcoat Value}
   defaultValue=<rgba 1 0 0 0>
  }
  n91 = VectorParameter() {group=4_Layer SFX, parameterName=SFX_Clearcoat_Ch_4, sortPriority=-4, defaultValue=<rgba 0 0 0 0.5>}
  n92 = ScalarParameter() {desc=If you need to cancel shared Mask texture on specific layer, group=4_Layer SFX, parameterName=SFX_OpacityMaskCancel_4, sortPriority=3}
  n93 = VectorParameter() {
   desc="Emissive color, by default mult by Base color. You can control that blending via Alpha channel of this input. Uses SFX texture as the Mask"
   group=4_Layer SFX
   parameterName="SFX_Emissive_Tint(Expand)_4"
   sortPriority=-6
   channelNames {a=Base color to Emissive, b="", g="", r=""}
   defaultValue=<rgba 0 0 0 1>
  }
  NamedRerouteDeclaration(None=MaterialFunctionCall!(
    WSM Scale=@n72.WSM 4
    Use WS Mapping=@n73
    Use Z=StaticSwitch(True=@n74, False=@n75, Value=@MaterialExpressionReroute_16.Output)
    Use XYZ=StaticSwitch(True=@n79, False=@n80, Value=@MaterialExpressionReroute_16.Output)
    WSM_ProjectionTransitionContrast
    UV2 True=@n71
    WSM_HeightLerpUsingAlpha
    WSM_HeightLerpUsingAlpha_Contrast
    UV's=@n72.WSM 4
    Normals_NN=@n81
    SpecularMasks_NN=@n82
    Diffuse_NN=@n83
    Use SFX Emissive=StaticBoolParameter() {desc=Enables SFX, group=4_Layer SFX, parameterName=SFX Emissive 4, sortPriority=-8}
    SFX Opacity Mask=@n84
    SFX SSS Two Sided Foliange=StaticBoolParameter() {group=4_Layer SFX, parameterName=SFX_SSS Foliage 4, sortPriority=-2}
    SFX SSS=StaticBoolParameter() {group=4_Layer SFX, parameterName=SFX_SSS 4, sortPriority=-2}
    SFX Clearcoat=StaticBoolParameter() {group=4_Layer SFX, parameterName="SFX_Clearcoat  4", sortPriority=-5}
    Diffuse=NamedRerouteDeclaration(None=@n86) {name=D4, nodeColor=<rgba 0 1 0.1797 1>}
    World Position Z=ScalarParameter() {group=4_layer UV, parameterName=L4 WP Offset Z, sortPriority=1}
    SFX Texture Complete=NamedRerouteUsage()
    SFX Tint Channel=Constant3Vector()
    SpecularMasks=NamedRerouteDeclaration(None=@n87) {name=S4, nodeColor=<rgba 1 0.5625 0 1>}
    Normals=NamedRerouteDeclaration(None=@n88) {name=N4, nodeColor=<rgba 0.7812 1 0 1>}
    Normals Power=ScalarParameter() {defaultValue=1, group=4_Layer, parameterName=NormalsValue_4, sortPriority=1}
    "SFX  SSS Channel"=VectorParameter() {group=4_Layer SFX, parameterName=SFX_SSS_Ch_4, sortPriority=-1}.RGB
    SSS Color Value=AppendVector(A=@n89, B=@n89)
    "Clearcoat (R) CC Roughness (G)"=@n90
    Clearcoat Mask Channal=@n91.RGB
    SFX Opacilty Mask Channel=VectorParameter() {group=4_Layer SFX, parameterName=SFX_OpacityMask_Ch_4, sortPriority=2}.RGB
    Opacilty Mask Cancel=@n92
    "SFX Emissive  Channel"=VectorParameter() {group=4_Layer SFX, parameterName=SFX_Emissive_Ch_4, sortPriority=-7}.RGB
    Emissive Tint=AppendVector(A=@n93, B=@n93)
    Data Ramp RGB
    Output Diffuse Only=@MaterialExpressionReroute_16.Output
    Output Normals Only=@MaterialExpressionReroute_17.Output2
   ) {materialFunction=/Game/Global/Base4Layers_2022/MaterialFunctions/MF_CombineLayers}.Result) {name=L4, nodeColor=<rgba 0.6172 0 1 1>}
  n94 = StaticBoolParameter() {group=3_layer UV, parameterName=Use UV2 for Layer 3}
  n95 = MaterialFunctionCall(L3 UV2=@n94) {materialFunction=&MF_UVsSet_New}
  n96 = StaticBoolParameter() {desc=Enable to use World Space Mapping for layer. Only XY or Z available in Layer 3, group=3_layer UV, parameterName=Layer 3 WS Mapping}
  n97 = StaticBoolParameter() {
   desc="World Space Z only, you need appropriate mask to make it visible. Use Z or map your layers accordingly "
   group=3_layer UV
   parameterName=Layer 3 WSM_Z
  }
  n98 = StaticSwitch(True=@n97, False=StaticBool() {value=true}, Value=StaticBoolParameter() {group=4_Layer, parameterName=Use 4 Layers, sortPriority=-1})
  n99 = StaticBool()
  n100 = StaticBoolParameter() {group=3_Layer_Blending, parameterName=Blend Overlay Layer 3}
  n101 = StaticSwitch(
   True=StaticBoolParameter() {desc=Use SFX to mask needed area, group=3_Layer_Blending, parameterName=L3 Simple Swatch, sortPriority=-15}
   False=@n99
   Value=@n100
  )
  MaterialExpressionReroute_14 = Reroute(None=@n101)
  n102 = StaticBoolParameter() {desc="World Space XYZ ( Expensive)", group=3_layer UV, parameterName=Layer 3 WSM_XYZ}
  n103 = StaticBool()
  n104 = TextureObjectParameter() {
   group=3_Layer_Nanite
   parameterName="Normals  Nanite 3"
   samplerType=SAMPLERTYPE_Normal
   sortPriority=0
   texture=&FlatNormal
  }
  n105 = TextureObjectParameter() {
   group=3_Layer_Nanite
   parameterName="SpecularMasks  Nanite 3"
   samplerType=SAMPLERTYPE_Masks
   sortPriority=0
   texture=&T_Fortnite_Default_S
  }
  n106 = TextureObjectParameter() {group=3_Layer_Nanite, parameterName=Diffuse Nanite 3, sortPriority=0, texture=&T_MidgreysRGB}
  n107 = StaticBoolParameter() {desc="Enables SFX for Opacity Mask, mast have Masked blending set in MI", group=3_Layer SFX, parameterName=SFX OpacityMask 3, sortPriority=1}
  n108 = StaticSwitch(True=StaticBoolParameter() {group=3_Layer_Blending, parameterName="L3  Normals Only", sortPriority=-15}, False=@n99, Value=@n100)
  MaterialExpressionReroute_15 = Reroute(None=@n108)
  n109 = TextureObjectParameter() {group=3_Layer, parameterName=Diffuse_Texture_3, sortPriority=0, texture=&T_MidgreysRGB}
  n110 = VectorParameter() {
   desc=Use R or G or B to select your SFX Mask Channel
   group=0_SFX BP
   parameterName=SFX_Tint_Ch_3
   sortPriority=4
   defaultValue=<rgba 1 0 0 0>
  }
  n111 = TextureObjectParameter() {
   desc="R = Specular\r\nG = Metal\r\nB = Roughness"
   group=3_Layer
   parameterName=SpecularMasks_3
   samplerType=SAMPLERTYPE_Masks
   sortPriority=0
   texture=&T_Fortnite_Default_S
  }
  n112 = TextureObjectParameter() {group=3_Layer, parameterName=Normals_Texture_3, samplerType=SAMPLERTYPE_Normal, sortPriority=0, texture=&FlatNormal}
  n113 = VectorParameter() {
   group=3_Layer SFX
   parameterName="SFX_SSS Tint(Expand)_3"
   primitiveDataIndex=2
   sortPriority=0
   channelNames {a=SSS Depth 0-1, b="", g="", r=""}
   defaultValue=<rgba 0 0 0 0.5>
  }
  n114 = VectorParameter() {
   desc="Use R and G channels of this input to controll ClearCoat amount and Rounghness. It uses SFX texture to mask area affected. "
   group=3_Layer SFX
   parameterName="SFX_ClearCoat(Expand)_3"
   sortPriority=-3
   channelNames {a=Not Used, b=Not Used, g=Clearcoat Roughness, r=Clearcoat Value}
   defaultValue=<rgba 1 0 0 0>
  }
  n115 = ScalarParameter() {desc=If you need to cancel shared Mask texture on specific layer, group=3_Layer SFX, parameterName=SFX_OpacityMaskCancel_3, sortPriority=3}
  n116 = VectorParameter() {
   desc="Emissive color, by default mult by Base color. You can control that blending via Alpha channel of this input. Uses SFX texture as the Mask"
   group=3_Layer SFX
   parameterName="SFX_Emissive_Tint(Expand)_3"
   sortPriority=-6
   channelNames {a=Base color to Emissive, b="", g="", r=""}
   defaultValue=<rgba 0 0 0 1>
  }
  NamedRerouteDeclaration(None=MaterialFunctionCall(
    WSM Scale=@n95.WSM 3
    Use WS Mapping=@n96
    Use Z=StaticSwitch(True=@n97, False=@n98, Value=@MaterialExpressionReroute_14.Output2)
    Use XYZ=StaticSwitch(True=@n102, False=@n103, Value=@MaterialExpressionReroute_14.Output2)
    UV2 True=@n94
    UV's=@n95.WSM 3
    Normals_NN=@n104
    SpecularMasks_NN=@n105
    Diffuse_NN=@n106
    Use SFX Emissive=StaticBoolParameter() {desc=Enables SFX, group=3_Layer SFX, parameterName=SFX Emissive 3, sortPriority=-8}
    SFX Opacity Mask=@n107
    SFX SSS Two Sided Foliange=StaticBoolParameter() {group=3_Layer SFX, parameterName=SFX_SSS Foliage 3, sortPriority=-2}
    SFX SSS=StaticBoolParameter() {group=3_Layer SFX, parameterName=SFX_SSS 3, sortPriority=-2}
    SFX Clearcoat=StaticBoolParameter() {group=3_Layer SFX, parameterName="SFX_Clearcoat  3", sortPriority=-5}
    Diffuse=NamedRerouteDeclaration(None=@n109) {name=D3, nodeColor=<rgba 0.9688 1 0 1>}
    World Position Z=ScalarParameter() {group=3_layer UV, parameterName=L3 WP Offset Z, sortPriority=1}
    SFX Texture Complete=NamedRerouteUsage()
    SFX Tint Channel=@n110.RGB
    SpecularMasks=NamedRerouteDeclaration(None=@n111) {name=S3, nodeColor=<rgba 0.2656 1 0 1>}
    Normals=NamedRerouteDeclaration(None=@n112) {name=N3, nodeColor=<rgba 0.8047 0 1 1>}
    Normals Power=ScalarParameter() {defaultValue=1, group=3_Layer, parameterName=NormalsValue_3, sortPriority=1}
    "SFX  SSS Channel"=VectorParameter() {group=3_Layer SFX, parameterName=SFX_SSS_Ch_3, sortPriority=-1}.RGB
    SSS Color Value=@n113
    "Clearcoat (R) CC Roughness (G)"=@n114
    Clearcoat Mask Channal=VectorParameter() {group=3_Layer SFX, parameterName=SFX_Clearcoat_Ch_3, sortPriority=-4}.RGB
    SFX Opacilty Mask Channel=VectorParameter() {group=3_Layer SFX, parameterName=SFX_OpacityMask_Ch_3, sortPriority=2}.RGB
    Opacilty Mask Cancel=@n115
    "SFX Emissive  Channel"=VectorParameter() {group=3_Layer SFX, parameterName=SFX_Emissive_Ch_3, sortPriority=-7}.RGB
    Emissive Tint=AppendVector(A=@n116, B=@n116)
    Output Diffuse Only=@MaterialExpressionReroute_14.Output2
    Output Normals Only=@MaterialExpressionReroute_15.Output3
   ) {materialFunction=&MF_CombineLayers}.Result) {name=L3, nodeColor=<rgba 0.6406 1 0 1>}
  n117 = TextureObjectParameter() {
   group=1_Layer_Nanite
   parameterName="Normals  Nanite"
   samplerType=SAMPLERTYPE_Normal
   sortPriority=0
   texture=&FlatNormal
  }
  n118 = TextureObjectParameter() {
   group=1_Layer_Nanite
   parameterName="SpecularMasks  Nanite"
   samplerType=SAMPLERTYPE_Masks
   sortPriority=0
   texture=&T_Fortnite_Default_S
  }
  n119 = TextureObjectParameter() {group=1_Layer_Nanite, parameterName=Diffuse Nanite, sortPriority=0, texture=&T_MidgreysRGB}
  n120 = StaticBoolParameter() {desc="Enables SFX for Opacity Mask, mast have Masked blending set in MI", group=1_Layer SFX, parameterName=SFX OpacityMask, sortPriority=1}
  n121 = TextureObjectParameter() {group=1_Layer, parameterName=Diffuse, sortPriority=0, texture=&T_MidgreysRGB}
  n122 = VectorParameter() {desc=Use R or G or B to select your SFX Mask Channel, group=0_SFX BP, parameterName=SFX_Tint_Ch, sortPriority=4, defaultValue=<rgba 1 0 0 0>}
  n123 = TextureObjectParameter() {
   desc="R = Specular\r\nG = Metal\r\nB = Roughness"
   group=1_Layer
   parameterName=SpecularMasks
   samplerType=SAMPLERTYPE_Masks
   sortPriority=0
   texture=&T_Fortnite_Default_S
  }
  n124 = TextureObjectParameter() {group=1_Layer, parameterName=Normals, samplerType=SAMPLERTYPE_Normal, sortPriority=0, texture=&FlatNormal}
  n125 = VectorParameter() {
   group=1_Layer SFX
   parameterName="SFX_SSS Tint(Expand)"
   sortPriority=0
   channelNames {a=SSS Depth 0-1, b="", g="", r=""}
   defaultValue=<rgba 0 0 0 0.5>
  }
  n126 = VectorParameter() {
   desc=Use R and G channels of this input to controll ClearCoat amount and Rounghness. It uses SFX texture to mask area affected.
   group=1_Layer SFX
   parameterName="SFX_ClearCoat(Expand)"
   sortPriority=-3
   channelNames {a=Not Used, b=Not Used, g=Clearcoat Roughness, r=Clearcoat Value}
   defaultValue=<rgba 1 0 0 0>
  }
  n127 = ScalarParameter() {desc=If you need to cancel shared Mask texture on specific layer, group=1_Layer SFX, parameterName=SFX_OpacityMaskCancel, sortPriority=3}
  n128 = VectorParameter() {
   desc="Emissive color, by default mult by Base color. You can control that blending via Alpha channel of this input. Uses SFX texture as the Mask"
   group=1_Layer SFX
   parameterName="SFX_Emissive_Tint(Expand)"
   sortPriority=-6
   channelNames {a=Base color to Emissive, b="", g="", r=""}
   defaultValue=<rgba 0 0 0 1>
  }
  n129 = ScalarParameter() {
   bUseCustomPrimitiveData=true
   desc="For now Scalar, this is possible alternative to Texture Swatch. Uses single ramp texture shared between all 4 layers and can have indepent row selected per layer. Enabled globally via function. Will disable use of color swatch if enabled. Diffuse tint still can be used if needed via separate mask "
   group=Y_CPD
   parameterName=CPD_PackHSV_1
   primitiveDataIndex=30
   sliderMin=1
  }
  NamedRerouteDeclaration(None=MaterialFunctionCall(
    UV's=MaterialFunctionCall() {materialFunction=&MF_UVsSet_New}.UVs Layer 1
    Normals_NN=@n117
    SpecularMasks_NN=@n118
    Diffuse_NN=@n119
    Use SFX Emissive=StaticBoolParameter() {desc=Enables SFX, group=1_Layer SFX, parameterName=SFX Emissive, sortPriority=-8}
    SFX Opacity Mask=@n120
    SFX SSS Two Sided Foliange=StaticBoolParameter() {group=1_Layer SFX, parameterName=SFX_SSS Foliage, sortPriority=-2}
    SFX SSS=StaticBoolParameter() {group=1_Layer SFX, parameterName=SFX_SSS, sortPriority=-2}
    SFX Clearcoat=StaticBoolParameter() {group=1_Layer SFX, parameterName=SFX_Clearcoat, sortPriority=-5}
    Diffuse=NamedRerouteDeclaration(None=@n121) {name=L1D, nodeColor=<rgba 0 0.2969 1 1>}
    World Position Z=Constant() {r=1}
    SFX Texture Complete=NamedRerouteUsage()
    SFX Tint Channel=@n122.RGB
    SpecularMasks=NamedRerouteDeclaration(None=@n123) {name=L1S, nodeColor=<rgba 1 0.7734 0 1>}
    Normals=NamedRerouteDeclaration(None=@n124) {name=L1N, nodeColor=<rgba 1 0.3984 0 1>}
    Normals Power=ScalarParameter() {defaultValue=1, group=1_Layer, parameterName=NormalsValue, sortPriority=1}
    "SFX  SSS Channel"=VectorParameter() {group=1_Layer SFX, parameterName=SFX_SSS_Ch, sortPriority=-1}.RGB
    SSS Color Value=AppendVector(A=@n125, B=@n125)
    "Clearcoat (R) CC Roughness (G)"=@n126
    Clearcoat Mask Channal=VectorParameter() {group=1_Layer SFX, parameterName=SFX_Clearcoat_Chl, sortPriority=-4}.RGB
    SFX Opacilty Mask Channel=VectorParameter() {group=1_Layer SFX, parameterName=SFX_OpacityMask_Ch_1, sortPriority=2}.RGB
    Opacilty Mask Cancel=@n127
    "SFX Emissive  Channel"=VectorParameter() {group=1_Layer SFX, parameterName=SFX_Emissive_Ch, sortPriority=-7}.RGB
    Emissive Tint=AppendVector(A=@n128, B=@n128)
    Data Ramp RGB=MaterialFunctionCall!(In=@n129) {materialFunction=/Game/Global/Base4Layers_2022/MaterialFunctions/MF_ColorTint_Unpack}.Result
    Output Diffuse Only=StaticBoolParameter() {group=1_Layer, parameterName=Layer 1 Diffuse Only}
   ) {materialFunction=&MF_CombineLayers}.Result) {name=L1, nodeColor=<rgba 0 0.5078 1 1>}
  n130 = StaticBoolParameter() {
   desc=Enable to use UV2 for mapping. You may need to switch masks for ohter layers To UV3
   group=2_layer UV
   parameterName=Use UV2 for Layer 2
  }
  n131 = MaterialFunctionCall(L2 UV2=@n130) {materialFunction=&MF_UVsSet_New}
  n132 = StaticBoolParameter() {
   desc="World Space Z only, you need appropriate mask to make it visible. Use Z or map your layers accordingly "
   group=2_layer UV
   parameterName=Layer 2 WSM_Z
  }
  n133 = TextureObjectParameter() {
   group=2_Layer_Nanite
   parameterName="Normals  Nanite 2"
   samplerType=SAMPLERTYPE_Normal
   sortPriority=0
   texture=&FlatNormal
  }
  n134 = TextureObjectParameter() {
   group=2_Layer_Nanite
   parameterName="SpecularMasks  Nanite 2"
   samplerType=SAMPLERTYPE_Masks
   sortPriority=0
   texture=&T_Fortnite_Default_S
  }
  n135 = TextureObjectParameter() {group=2_Layer_Nanite, parameterName=Diffuse Nanite 2, sortPriority=0, texture=&T_MidgreysRGB}
  n136 = ScalarParameter() {
   bUseCustomPrimitiveData=true
   desc="For now Scalar, this is possible alternative to Texture Swatch. Uses single ramp texture shared between all 4 layers and can have indepent row selected per layer. Enabled globally via function. Will disable use of color swatch if enabled. Diffuse tint still can be used if needed via separate mask "
   group=Y_CPD
   parameterName=CPD_PackedHSV_2
   primitiveDataIndex=31
   sliderMin=1
  }
  n137 = StaticBool()
  n138 = StaticBoolParameter() {group=2_Layer_Blending, parameterName=Blend Overlay Layer 2}
  n139 = StaticSwitch(True=StaticBoolParameter() {group=2_Layer_Blending, parameterName=L2 Simple Swatch, sortPriority=-15}, False=@n137, Value=@n138)
  MaterialExpressionReroute_11 = Reroute(None=@n139)
  n140 = StaticSwitch(True=StaticBoolParameter() {group=2_Layer_Blending, parameterName=L2 Normals Only, sortPriority=-15}, False=@n137, Value=@n138)
  MaterialExpressionReroute_12 = Reroute(None=@n140)
  n141 = VectorParameter() {
   group=2_Layer SFX
   parameterName="SFX_SSS Tint (Expand)_2"
   primitiveDataIndex=2
   sortPriority=0
   channelNames {a=SSS Depth 0-1, b="", g="", r=""}
   defaultValue=<rgba 0 0 0 0.5>
  }
  n142 = VectorParameter() {
   desc="Use R and G channels of this input to controll ClearCoat amount and Rounghness. It uses SFX texture to mask area affected. "
   group=2_Layer SFX
   parameterName="SFX_ClearCoat(Expand)_2"
   sortPriority=-3
   channelNames {a=Not Used, b=Not Used, g=Clearcoat Roughness, r=Clearcoat Value}
   defaultValue=<rgba 1 0 0 0>
  }
  n143 = ScalarParameter() {desc=If you need to cancel shared Mask texture on specific layer, group=2_Layer SFX, parameterName=SFX_OpacityMaskCancel_2, sortPriority=3}
  n144 = VectorParameter() {
   desc="Emissive color, by default mult by Base color. You can control that blending via Alpha channel of this input. Uses SFX texture as the Mask"
   group=2_Layer SFX
   parameterName="SFX_Emissive_Tint(Expand)_2"
   sortPriority=-6
   channelNames {a=Base color to Emissive, b="", g="", r=""}
   defaultValue=<rgba 0 0 0 1>
  }
  n145 = TextureObjectParameter() {group=2_Layer, parameterName=Diffuse_Texture_2, sortPriority=0, texture=&T_MidgreysRGB}
  n146 = VectorParameter() {
   desc=Use R or G or B to select your SFX Mask Channel
   group=0_SFX BP
   parameterName=SFX_Tint_Ch_2
   sortPriority=4
   defaultValue=<rgba 1 0 0 0>
  }
  n147 = TextureObjectParameter() {
   desc="R = Specular\r\nG = Metal\r\nB = Roughness"
   group=2_Layer
   parameterName=SpecularMasks_2
   samplerType=SAMPLERTYPE_Masks
   sortPriority=0
   texture=&T_Fortnite_Default_S
  }
  n148 = TextureObjectParameter() {group=2_Layer, parameterName=Normals_Texture_2, samplerType=SAMPLERTYPE_Normal, sortPriority=0, texture=&FlatNormal}
  NamedRerouteDeclaration(None=MaterialFunctionCall(
    WSM Scale=@n131.WSM 2
    Use WS Mapping=StaticBoolParameter() {desc=Enable to use World Space Mapping for layer, group=2_layer UV, parameterName=Layer 2 WS Mapping}
    Use Z=@n132
    Use XYZ=StaticBoolParameter() {desc="World Space XYZ ( Expensive)", group=2_layer UV, parameterName=Layer 2 WSM_XYZ}
    UV2 True=@n130
    UV's=@n131.WSM 2
    Normals_NN=@n133
    SpecularMasks_NN=@n134
    Diffuse_NN=@n135
    Use SFX Emissive=StaticBoolParameter() {desc=Enables SFX, group=2_Layer SFX, parameterName=SFX Emissive 2, sortPriority=-8}
    SFX Opacity Mask=StaticBoolParameter() {desc=Enables SFX, group=2_Layer SFX, parameterName=SFX OpacityMask 2, sortPriority=1}
    SFX SSS Two Sided Foliange=StaticBoolParameter() {group=2_Layer SFX, parameterName=SFX_SSS Foliage 2, sortPriority=-2}
    SFX SSS=StaticBoolParameter() {group=2_Layer SFX, parameterName=SFX_SSS 2, sortPriority=-2}
    SFX Clearcoat=StaticBoolParameter() {group=2_Layer SFX, parameterName="SFX_Clearcoat  2", sortPriority=-5}
    Diffuse=NamedRerouteDeclaration(None=@n145) {name=D2, nodeColor=<rgba 1 0.3984 0 1>}
    World Position Z=ScalarParameter() {group=2_layer UV, parameterName=L2 WP Offset Z, sortPriority=1}
    SFX Texture Complete=NamedRerouteUsage()
    SFX Tint Channel=@n146.RGB
    SpecularMasks=NamedRerouteDeclaration(None=@n147) {name=S2, nodeColor=<rgba 1 0 0.3281 1>}
    Normals=NamedRerouteDeclaration(None=@n148) {name=N2, nodeColor=<rgba 0 0.5547 1 1>}
    Normals Power=ScalarParameter() {defaultValue=1, group=2_Layer, parameterName=NormalsValue_2, sortPriority=1}
    "SFX  SSS Channel"=VectorParameter() {group=2_Layer SFX, parameterName=SFX_SSS_Ch_2, sortPriority=-1}.RGB
    SSS Color Value=AppendVector(A=@n141, B=@n141)
    "Clearcoat (R) CC Roughness (G)"=@n142
    Clearcoat Mask Channal=VectorParameter() {group=2_Layer SFX, parameterName=SFX_Clearcoat_Ch_2, sortPriority=-4}.RGB
    SFX Opacilty Mask Channel=VectorParameter() {group=2_Layer SFX, parameterName=SFX_OpacityMask_Ch_2, sortPriority=2}.RGB
    Opacilty Mask Cancel=@n143
    "SFX Emissive  Channel"=VectorParameter() {group=2_Layer SFX, parameterName=SFX_Emissive_Ch_2, sortPriority=-7}.RGB
    Emissive Tint=AppendVector(A=@n144, B=@n144)
    Data Ramp RGB=MaterialFunctionCall(In=@n136) {materialFunction=&MF_ColorTint_Unpack}.Result
    Output Diffuse Only=@MaterialExpressionReroute_11.Output
    Output Normals Only=@MaterialExpressionReroute_12.Output2
   ) {materialFunction=&MF_CombineLayers}.Result) {name=L2, nodeColor=<rgba 1 0 0.5859 1>}
  NamedRerouteDeclaration(None=@n10) {name=SFX Texture Global, nodeColor=<rgba 0.3594 1 0 1>}
  NamedRerouteDeclaration(None=@n68.Base Color) {name=Base_BaseColorIn, nodeColor=<rgba 1 0 0.2578 1>}
  n149 = TextureObjectParameter() {
   desc="SFX RGB - Graffiti, emissive RGB "
   group=0_SFX Global
   parameterName=SFX_RGB
   sortPriority=1
   texture=/Game/Global/Base4Layers_2022/Textures/T-GraffitiSample_D
  }
  NamedRerouteDeclaration(None=@n149) {name=SFX Texture RGB, nodeColor=<rgba 0 0.5312 1 1>}
  n150 = NamedRerouteUsage()
  n151 = ScalarParameter() {defaultValue=0.3, parameterName=BaseColorDarkening}
  n152 = TextureSample() {samplerType=SAMPLERTYPE_Masks, texture=/FirePetalShrineGameplay/Assets/Textures/T_ClayTile_BoonObelisk_A_A_M2}
  NamedRerouteDeclaration(None=LinearInterpolate(
    A=@n150
    B=Multiply(A=@n150, B=@n151) {constB=0.2}
    Alpha=NamedRerouteDeclaration(None=@n152.R) {name=RuneMask, nodeColor=<rgba 1 0 0.2344 1>}
   ) {constB=0.2}) {name=Base_BaseColorOut, nodeColor=<rgba 0 0.5781 1 1>}
  n153 = ScalarParameter() {bUseCustomPrimitiveData=true, parameterName=TimeInput}
  n154 = Divide(A=Subtract(A=Time(), B=@n153), B=ScalarParameter() {defaultValue=15, parameterName=Duration})
  NamedRerouteDeclaration(None=Multiply(A=Saturate(None=@n154), B=Sign(None=@n153))) {name=AnimationProgress, nodeColor=<rgba 0 1 0.0859 1>}
  n155 = NamedRerouteUsage()
  n156 = Multiply(A=NamedRerouteUsage(), B=ScalarParameter() {defaultValue=0.1, parameterName=NoiseAmount})
  n157 = SphereMask(A=Add(A=ComponentMask(None=@n156) {g=true, r=true}, B=TextureCoordinate()), B=Constant2Vector() {g=0.32, r=0.32}, Radius=NamedRerouteUsage()) {attenuationRadius=0.02, hardnessPercent=0}
  n158 = NamedRerouteUsage()
  n159 = Sign(None=ScalarParameter() {bUseCustomPrimitiveData=true, parameterName=TimeInput})
  n160 = TextureCoordinate()
  n161 = Constant2Vector() {g=0.25, r=0.25}
  n162 = Frac(None=Time())
  n163 = OneMinus(None=MaterialFunctionCall(Vector or UVs=MaterialFunctionCall(UVs=@n160, Rotation Center=@n161, "Rotation Angle (0-1)"=@n162) {materialFunction=&CustomRotator}.Rotated Values) {materialFunction=&VectorToRadialValue}.Linear Distance)
  n164 = Saturate(None=Multiply(A=Power(Base=@n163) {constExponent=12}) {constB=5})
  n165 = Add(A=Multiply(A=Multiply(A=@n157) {constB=5}, B=SmoothStep(Value=@n158) {constMin=0.6}), B=Multiply(A=@n158, B=Multiply(A=@n159, B=@n164)))
  NamedRerouteDeclaration(None=Multiply(A=ScalarParameter() {defaultValue=0.02, parameterName=EmissiveMult}, B=Multiply(A=@n155, B=@n165))) {name=Base_EmissiveOut, nodeColor=<rgba 0.3594 1 0 1>}
  PinBase!(Output=@n62) {reroutePins [Output {expression=&self:MaterialExpressionReroute_9}]}
  PinBase!() {reroutePins [Input {expression=&self:MaterialExpressionReroute_3}]}
  PinBase()
  PinBase!(Output=@n139, Output2=@n140) {reroutePins [
    Output {expression=&self:MaterialExpressionReroute_11}
    Output2 {expression=&self:MaterialExpressionReroute_12}
   ]}
  PinBase!()
  PinBase!(Output2=@n101, Output3=@n108) {reroutePins [
    Output2 {expression=&self:MaterialExpressionReroute_14}
    Output3 {expression=&self:MaterialExpressionReroute_15}
   ]}
  PinBase!()
  PinBase!(Output=@n78, Output2=@n85) {reroutePins [
    Output {expression=&self:MaterialExpressionReroute_16}
    Output2 {expression=&self:MaterialExpressionReroute_17}
   ]}
  Reroute(None=@n51)
  Saturate(None=Add(A=NamedRerouteUsage(), B=NamedRerouteUsage()))
  n166 = GetMaterialAttributes(None=@n67.POI Color) {attributeGetTypes [#4]}
  n167 = NamedRerouteUsage()
  n168 = LinearInterpolate(A=Add(A=@n166.Emissive Color, B=Add(A=@n167, B=@n4)), Alpha=@n5) {constB=0}
  n169 = NamedRerouteUsage()
  StaticSwitchParameter(
   True=SetMaterialAttributes!(
    MaterialAttributes=@n166.MaterialAttributes
    Base Color=NamedRerouteUsage()
    World Position Offset=Constant()
    Emissive Color=NamedRerouteUsage()
   ) {attributeSetTypes [#3, #5, #4]}
   False=SetMaterialAttributes!(MaterialAttributes=@n67.POI Color, Emissive Color=Add(A=@n168, B=@n169), World Position Offset=NamedRerouteUsage()) {attributeSetTypes [#4, #5]}
  ) {parameterName=IsBase}
 ]
}
